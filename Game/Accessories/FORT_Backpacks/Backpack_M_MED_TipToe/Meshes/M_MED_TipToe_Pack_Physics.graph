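# Physics asset for the M_MED_TipToe backpack, giving its pack and ten arrow bones simple bodies for physical motion.
# Each arrow has a small identical capsule and its own constraint directly to the pack capsule, forming a star layout with no arrow-to-arrow links.
# Constraints carry tight angular limits, and several lock one or both swing axes, so freedom of movement differs between arrows.
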PhysicsAsset /Game/Accessories/FORT_Backpacks/Backpack_M_MED_TipToe/Meshes/M_MED_TipToe_Pack_Physics
bodies {
 arrow1 [{shapeType=Capsule, length=1.5, radius=1, center=<1 0 0>, rotation=<rot 90 0 0>}]
 arrow10 [{shapeType=Capsule, length=1.5, radius=1, center=<1 0 0>, rotation=<rot 90 0 0>}]
 arrow2 [{shapeType=Capsule, length=1.5, radius=1, center=<1 0 0>, rotation=<rot 90 0 0>}]
 arrow3 [{shapeType=Capsule, length=1.5, radius=1, center=<1 0 0>, rotation=<rot 90 0 0>}]
 arrow4 [{shapeType=Capsule, length=1.5, radius=1, center=<1 0 0>, rotation=<rot 90 0 0>}]
 arrow5 [{shapeType=Capsule, length=1.5, radius=1, center=<1 0 0>, rotation=<rot 90 0 0>}]
 arrow6 [{shapeType=Capsule, length=1.5, radius=1, center=<1 0 0>, rotation=<rot 90 0 0>}]
 arrow7 [{shapeType=Capsule, length=1.5, radius=1, center=<1 0 0>, rotation=<rot 90 0 0>}]
 arrow8 [{shapeType=Capsule, length=1.5, radius=1, center=<1 0 0>, rotation=<rot 90 0 0>}]
 arrow9 [{shapeType=Capsule, length=1.5, radius=1, center=<1 0 0>, rotation=<rot 90 0 0>}]
 pack [{shapeType=Capsule, length=1, radius=1}]
}
constraints [
 {bone1Name=arrow1, bone2Name=pack, swing1LimitDegrees=5, swing1Motion=Locked, swing2LimitDegrees=5, swing2Motion=Locked, twistLimitDegrees=5}
 {bone1Name=arrow2, bone2Name=pack, swing1LimitDegrees=5, swing2LimitDegrees=5, twistLimitDegrees=5}
 {bone1Name=arrow3, bone2Name=pack, swing1LimitDegrees=10, swing2LimitDegrees=10, twistLimitDegrees=5}
 {bone1Name=arrow4, bone2Name=pack, swing1LimitDegrees=5, swing2LimitDegrees=5, twistLimitDegrees=5}
 {bone1Name=arrow5, bone2Name=pack, swing1LimitDegrees=5, swing1Motion=Locked, swing2LimitDegrees=5, twistLimitDegrees=5}
 {bone1Name=arrow6, bone2Name=pack, swing1LimitDegrees=5, swing2LimitDegrees=5, twistLimitDegrees=5}
 {bone1Name=arrow7, bone2Name=pack, swing1LimitDegrees=5, swing1Motion=Locked, swing2LimitDegrees=5, twistLimitDegrees=5}
 {bone1Name=arrow8, bone2Name=pack, swing1LimitDegrees=5, swing1Motion=Locked, swing2LimitDegrees=5, swing2Motion=Locked, twistLimitDegrees=5}
 {
  bone1Name=arrow9
  bone2Name=pack
  swing1LimitDegrees=10
  swing1Motion=Locked
  swing2LimitDegrees=10
  swing2Motion=Locked
  twistLimitDegrees=5
 }
 {
  bone1Name=arrow10
  bone2Name=pack
  swing1LimitDegrees=10
  swing1Motion=Locked
  swing2LimitDegrees=10
  swing2Motion=Locked
  twistLimitDegrees=5
 }
]
